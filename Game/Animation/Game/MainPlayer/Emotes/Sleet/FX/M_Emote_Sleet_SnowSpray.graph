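# Masked, unlit Niagara sprite material for the snow spray effect in the player's Sleet emote.
# A time-of-day function blends morning, day, evening and night tint parameters, and the result multiplies the particle color.
# Particle alpha minus the inverted alpha of a dirt-clod cluster texture forms an erosion-style mask, so lowering alpha shrinks the clumps.
# A matching dirt-clod normal map is also sampled, though the shading model is unlit.
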Material /Game/Animation/Game/MainPlayer/Emotes/Sleet/FX/M_Emote_Sleet_SnowSpray
graph {
 properties {bUsedWithNiagaraSprites=true, blendMode=BLEND_Masked, pixelDepthOffsetMode=PDOM_Legacy, shadingModel=MSM_Unlit, usedShadingModels=Unlit}
 nodes [
  n1 = ParticleColor()
  Multiply(
   A=MaterialFunctionCall(
    Color_Morning=VectorParameter() {parameterName=Color Morning Multiplier, defaultValue=<rgba 0.9896 0.268 0.4929 0>}.RGB
    Color_Day=VectorParameter() {parameterName=Color Day Multiplier, defaultValue=<rgba 1 1 1 0>}.RGB
    Color_Evening=VectorParameter() {parameterName=Color Evening Multiplier, defaultValue=<rgba 0.599 0.2901 0.3358 0>}.RGB
    Color_Night=VectorParameter() {parameterName=Color Night Multiplier, defaultValue=<rgba 0.3498 0.4864 0.6458 0>}.RGB
   ) {materialFunction=&Color_Time_of_Day}.Result
   B=@n1.RGB
  )
  Subtract(A=@n1.A, B=OneMinus(None=TextureSample() {texture=/Game/Effects/Fort_Effects/Textures/Dirt/T-DirtClodCluster_D}.A))
  TextureSample() {samplerType=SAMPLERTYPE_Normal, texture=/Game/Effects/Fort_Effects/Textures/Dirt/T-DirtClodCluster_N}
 ]
}
